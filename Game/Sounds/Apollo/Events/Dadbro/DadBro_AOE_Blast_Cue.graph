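# Sound cue for the DadBro AOE blast, built from Storm King lightning and shockwave sounds.
# A Mixer layers one of three equally weighted lightning blasts with a shockwave that plays louder after a short fixed delay.
# A root Modulator processes the mix, which routes to the player-critical boss fight sound class.
# Overridden attenuation disables distance falloff and selects HRTF spatialization.
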SoundCue /Game/Sounds/Apollo/Events/Dadbro/DadBro_AOE_Blast_Cue
root=Modulator_3
nodes [
 {id=Delay_0, children [WavePlayer_0], properties {delayMax=0.2, delayMin=0.2}}
 {id=Mixer_0, children [Random_0,Delay_0], properties {inputVolume [1,1.5]}}
 {id=Modulator_3, children [Mixer_0]}
 {id=Random_0, children [WavePlayer_2,WavePlayer_3,WavePlayer_4], properties {weights [1,1,1]}}
 {id=WavePlayer_0, properties {soundWaveAssetPtr=/Game/Sounds/Fort_Monster_Storm_King/SFX/Final_SFX/StormKing_AOE_Shockwave}}
 {id=WavePlayer_2, properties {soundWaveAssetPtr=/Game/Sounds/Fort_Monster_Storm_King/SFX/StormKing_LightningSummons_Blast_01}}
 {id=WavePlayer_3, properties {soundWaveAssetPtr=/Game/Sounds/Fort_Monster_Storm_King/SFX/StormKing_LightningSummons_Blast_02}}
 {id=WavePlayer_4, properties {soundWaveAssetPtr=/Game/Sounds/Fort_Monster_Storm_King/SFX/StormKing_LightningSummons_Blast_03}}
]
properties {
 bOverrideAttenuation=true
 maxDistance=2097152
 soundClassObject=/Game/Sounds/FortSoundClassesAndModes/Fort_BossFight_SoundClass/Fort_BossFight_PlayerCritical
 volumeMultiplier=1.25
 attenuationOverrides {bAttenuate=false, falloffDistance=10000, spatializationAlgorithm=SPATIALIZATION_HRTF, stereoSpread=1000, attenuationShapeExtents {x=10000}}
 concurrencyOverrides {resolutionRule=PreventNew}
}
