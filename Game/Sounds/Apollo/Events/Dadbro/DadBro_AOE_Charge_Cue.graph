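# Sound cue for the DadBro area-of-effect charge, reusing Storm King lightning-summon charge recordings.
# A Random node picks one of two equally weighted waves and feeds a Modulator at the root.
# The cue plays boosted in the boss-fight player-critical sound class, with distance attenuation disabled and HRTF set as the spatialization algorithm.
# Overridden concurrency allows one instance at a time and blocks new plays while it is active.
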SoundCue /Game/Sounds/Apollo/Events/Dadbro/DadBro_AOE_Charge_Cue
root=Modulator_3
nodes [
 {id=Modulator_3, children [Random_0]}
 {id=Random_0, children [WavePlayer_0,WavePlayer_1], properties {weights [1,1]}}
 {id=WavePlayer_0, properties {soundWaveAssetPtr=/Game/Sounds/Fort_Monster_Storm_King/SFX/Final_SFX/StormKing_LightningSummons_Charge_02}}
 {id=WavePlayer_1, properties {soundWaveAssetPtr=/Game/Sounds/Fort_Monster_Storm_King/SFX/Final_SFX/StormKing_LightningSummons_Charge_03}}
]
properties {
 bOverrideAttenuation=true
 bOverrideConcurrency=true
 maxDistance=2097152
 soundClassObject=/Game/Sounds/FortSoundClassesAndModes/Fort_BossFight_SoundClass/Fort_BossFight_PlayerCritical
 volumeMultiplier=1.5
 attenuationOverrides {bAttenuate=false, falloffDistance=10000, spatializationAlgorithm=SPATIALIZATION_HRTF, stereoSpread=1000, attenuationShapeExtents {x=10000}}
 concurrencyOverrides {maxCount=1, resolutionRule=PreventNew}
}
