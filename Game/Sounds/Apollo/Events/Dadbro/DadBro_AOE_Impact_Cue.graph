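# Sound cue for the impact of DadBro's area-of-effect attack, reusing Storm King lightning summon impact recordings.
# A Random node picks one of three equally weighted wave variants and feeds a Modulator with volume held at unity.
# The cue plays slightly boosted in the boss fight player-critical sound class.
# Its attenuation override disables distance falloff and selects HRTF spatialization.
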SoundCue /Game/Sounds/Apollo/Events/Dadbro/DadBro_AOE_Impact_Cue
root=Modulator_3
nodes [
 {id=Modulator_3, children [Random_1], properties {volumeMax=1, volumeMin=1}}
 {id=Random_1, children [WavePlayer_2,WavePlayer_3,WavePlayer_4], properties {weights [1,1,1]}}
 {id=WavePlayer_2, properties {soundWaveAssetPtr=/Game/Sounds/Fort_Monster_Storm_King/SFX/StormKing_LightningSummons_Impact_01}}
 {id=WavePlayer_3, properties {soundWaveAssetPtr=/Game/Sounds/Fort_Monster_Storm_King/SFX/StormKing_LightningSummons_Impact_02}}
 {id=WavePlayer_4, properties {soundWaveAssetPtr=/Game/Sounds/Fort_Monster_Storm_King/SFX/StormKing_LightningSummons_Impact_03}}
]
properties {
 bOverrideAttenuation=true
 maxDistance=2097152
 soundClassObject=/Game/Sounds/FortSoundClassesAndModes/Fort_BossFight_SoundClass/Fort_BossFight_PlayerCritical
 volumeMultiplier=1.2
 attenuationOverrides {bAttenuate=false, falloffDistance=100, spatializationAlgorithm=SPATIALIZATION_HRTF, stereoSpread=1000, attenuationShapeExtents {x=600}}
 concurrencyOverrides {resolutionRule=PreventNew}
}
