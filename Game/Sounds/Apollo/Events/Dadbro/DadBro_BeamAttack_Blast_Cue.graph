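# Sound cue for the DadBro boss beam attack blast, reusing two StormKing beam blast waves in the boss fight SFX class.
# A random node picks between the equally weighted waves and feeds a modulator that adds slight downward pitch variation.
# Cue-level overrides boost volume, disable distance attenuation while setting HRTF spatialization, and block new instances once the concurrency limit is reached.
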SoundCue /Game/Sounds/Apollo/Events/Dadbro/DadBro_BeamAttack_Blast_Cue
root=Modulator_0
nodes [
 {id=Modulator_0, children [Random_1], properties {pitchMax=1, pitchMin=0.98}}
 {id=Random_1, children [WavePlayer_1,WavePlayer_2], properties {weights [1,1]}}
 {id=WavePlayer_1, properties {soundWaveAssetPtr=&/StormKing_BeamAttack_Blast_05}}
 {id=WavePlayer_2, properties {soundWaveAssetPtr=&/StormKing_BeamAttack_Blast_06}}
]
properties {
 bOverrideAttenuation=true
 bOverrideConcurrency=true
 maxDistance=2097152
 soundClassObject=&Fort_BossFight_SFX
 volumeMultiplier=1.5
 attenuationOverrides {bAttenuate=false, falloffDistance=8000, spatializationAlgorithm=SPATIALIZATION_HRTF, attenuationShapeExtents {x=4000}}
 concurrencyOverrides {resolutionRule=PreventNew}
}
