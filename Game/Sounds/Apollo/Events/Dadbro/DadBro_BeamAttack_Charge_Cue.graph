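# Sound cue for the charge of DadBro's beam attack, routed to the boss fight SFX sound class with boosted volume.
# A Random node picks one of four equally weighted Storm King beam charge waves, reused from that monster's audio.
# An attenuation override turns off distance falloff and selects HRTF spatialization with a stereo spread.
# A concurrency override blocks new instances once the limit is reached.
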SoundCue /Game/Sounds/Apollo/Events/Dadbro/DadBro_BeamAttack_Charge_Cue
root=Random_0
nodes [
 {id=Random_0, children [WavePlayer_0,WavePlayer_2,WavePlayer_6,WavePlayer_7], properties {weights [1,1,1,1]}}
 {id=WavePlayer_0, properties {soundWaveAssetPtr=/Game/Sounds/Fort_Monster_Storm_King/SFX/Final_SFX/StormKing_BeamAttack_Charge_A_v2_04}}
 {id=WavePlayer_2, properties {soundWaveAssetPtr=/Game/Sounds/Fort_Monster_Storm_King/SFX/Final_SFX/StormKing_BeamAttack_Charge_A_v2_01}}
 {id=WavePlayer_6, properties {soundWaveAssetPtr=/Game/Sounds/Fort_Monster_Storm_King/SFX/Final_SFX/StormKing_BeamAttack_Charge_A_v2_02}}
 {id=WavePlayer_7, properties {soundWaveAssetPtr=/Game/Sounds/Fort_Monster_Storm_King/SFX/Final_SFX/StormKing_BeamAttack_Charge_A_v2_03}}
]
properties {
 bOverrideAttenuation=true
 bOverrideConcurrency=true
 maxDistance=2097152
 soundClassObject=&Fort_BossFight_SFX
 volumeMultiplier=1.5
 attenuationOverrides {bAttenuate=false, falloffDistance=10000, spatializationAlgorithm=SPATIALIZATION_HRTF, stereoSpread=1000, attenuationShapeExtents {x=3000}}
 concurrencyOverrides {resolutionRule=PreventNew}
}
