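# DadBro_Death_Mid_Cue is a sound cue, apparently for the mid stage of DadBro's death in an Apollo event, that reuses the Storm King death sound.
# A single wave player feeds the root at reduced volume, routed to a 2D hit-notification sound class.
# Attenuation override disables distance falloff, so it plays unattenuated regardless of position.
# Concurrency allows one instance and prevents new ones while it plays.
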SoundCue /Game/Sounds/Apollo/Events/Dadbro/DadBro_Death_Mid_Cue
root=WavePlayer_0
nodes [{id=WavePlayer_0, properties {soundWaveAssetPtr=/Game/Sounds/Fort_Monster_Storm_King/SFX/Final_SFX/StormKing_Death_v4}}]
properties {
 bOverrideAttenuation=true
 bOverrideConcurrency=true
 maxDistance=2097152
 soundClassObject=&Fort_2DHitNotifications_Instigate_PSM
 volumeMultiplier=0.5
 attenuationOverrides {bAttenuate=false, falloffDistance=20000, spatializationAlgorithm=SPATIALIZATION_HRTF, stereoSpread=1000, attenuationShapeExtents {x=10000}}
 concurrencyOverrides {maxCount=1, resolutionRule=PreventNew}
}
