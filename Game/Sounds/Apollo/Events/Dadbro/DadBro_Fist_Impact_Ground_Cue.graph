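# Sound cue for the DadBro fist ground impact in the boss-fight SFX class, reusing Storm King melee impact waves.
# A mixer layers a random pick of three stereo variants with a slightly quieter random pick of three non-stereo-named variants, then feeds a Modulator.
# It uses air-strike attenuation, and its concurrency override settings specify a single instance that prevents new plays.
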SoundCue /Game/Sounds/Apollo/Events/Dadbro/DadBro_Fist_Impact_Ground_Cue
root=Modulator_3
nodes [
 {id=Mixer_1, children [Random_1,Random_0], properties {inputVolume [1,0.8]}}
 {id=Modulator_3, children [Mixer_1]}
 {id=Random_0, children [WavePlayer_0,WavePlayer_1,WavePlayer_2], properties {weights [1,1,1]}}
 {id=Random_1, children [WavePlayer_3,WavePlayer_4,WavePlayer_5], properties {weights [1,1,1]}}
 {id=WavePlayer_0, properties {soundWaveAssetPtr=/Game/Sounds/Fort_Monster_Storm_King/SFX/Final_SFX/StormKing_Melee_Impact_01}}
 {id=WavePlayer_1, properties {soundWaveAssetPtr=/Game/Sounds/Fort_Monster_Storm_King/SFX/Final_SFX/StormKing_Melee_Impact_02}}
 {id=WavePlayer_2, properties {soundWaveAssetPtr=/Game/Sounds/Fort_Monster_Storm_King/SFX/Final_SFX/StormKing_Melee_Impact_03}}
 {id=WavePlayer_3, properties {soundWaveAssetPtr=/Game/Sounds/Fort_Monster_Storm_King/SFX/Final_SFX/StormKing_Melee_Impact_Stereo_01}}
 {id=WavePlayer_4, properties {soundWaveAssetPtr=/Game/Sounds/Fort_Monster_Storm_King/SFX/Final_SFX/StormKing_Melee_Impact_Stereo_02}}
 {id=WavePlayer_5, properties {soundWaveAssetPtr=/Game/Sounds/Fort_Monster_Storm_King/SFX/Final_SFX/StormKing_Melee_Impact_Stereo_03}}
]
properties {
 attenuationSettings=/Game/Sounds/Attenuation/Fort_AirStrike_Att
 maxDistance=9000
 soundClassObject=&Fort_BossFight_SFX
 volumeMultiplier=1
 concurrencyOverrides {maxCount=1, resolutionRule=PreventNew}
}
