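# Sound cue for the DadBro boss event, playing a Storm King back hit reaction for the hands-down state.
# A Random node with equal weights picks one of three hand hit-react waves.
# It routes to the boss fight SFX sound class and overrides its referenced Storm King vox attenuation with volume attenuation disabled, so distance does not reduce its loudness.
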SoundCue /Game/Sounds/Apollo/Events/Dadbro/DadBro_HandsDown_Back_React_Cue
root=Random_1
nodes [
 {id=Random_1, children [WavePlayer_0,WavePlayer_4,WavePlayer_5], properties {weights [1,1,1]}}
 {id=WavePlayer_0, properties {soundWaveAssetPtr=&/StormKing_HitReact_Back_Hand_01}}
 {id=WavePlayer_4, properties {soundWaveAssetPtr=&/StormKing_HitReact_Back_Hand_02}}
 {id=WavePlayer_5, properties {soundWaveAssetPtr=&/StormKing_HitReact_Back_Hand_03}}
]
properties {
 attenuationSettings=/Game/Sounds/Attenuation/Fort_StormKingVox_Att
 bOverrideAttenuation=true
 maxDistance=2097152
 soundClassObject=&Fort_BossFight_SFX
 volumeMultiplier=1
 attenuationOverrides {bAttenuate=false, lPFRadiusMin=17}
}
