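# Sound cue for DadBro's hands-down chest reaction, built from Storm King chest-hand hit-react waves.
# A Random node picks one of three equally weighted wave players.
# It plays in the boss fight SFX class, and its inline attenuation override disables volume falloff with distance, taking precedence over the assigned Storm King vox attenuation asset.
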SoundCue /Game/Sounds/Apollo/Events/Dadbro/DadBro_HandsDown_Chest_React_Cue
root=Random_0
nodes [
 {id=Random_0, children [WavePlayer_1,WavePlayer_2,WavePlayer_3], properties {weights [1,1,1]}}
 {id=WavePlayer_1, properties {soundWaveAssetPtr=&/StormKing_HitReact_Chest_Hand_01}}
 {id=WavePlayer_2, properties {soundWaveAssetPtr=&/StormKing_HitReact_Chest_Hand_02}}
 {id=WavePlayer_3, properties {soundWaveAssetPtr=&/StormKing_HitReact_Chest_Hand_03}}
]
properties {
 attenuationSettings=/Game/Sounds/Attenuation/Fort_StormKingVox_Att
 bOverrideAttenuation=true
 maxDistance=2097152
 soundClassObject=&Fort_BossFight_SFX
 volumeMultiplier=1
 attenuationOverrides {bAttenuate=false, lPFRadiusMin=17}
}
