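# Sound cue for the DadBro horn-destroyed sound in the Apollo event audio, reusing Storm King horn break recordings.
# A random node picks one of three equally weighted waves, wrapped in a modulator for variation and a very short fixed delay.
# Overridden attenuation turns off distance falloff, and the cue routes to an instigator 2D hit-notification sound class at reduced volume.
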SoundCue /Game/Sounds/Apollo/Events/Dadbro/DadBro_Horn_Destroyed_Cue
root=Delay_0
nodes [
 {id=Delay_0, children [Modulator_0], properties {delayMax=0.02, delayMin=0.02}}
 {id=Modulator_0, children [Random_0], properties {volumeMax=1}}
 {id=Random_0, children [WavePlayer_0,WavePlayer_1,WavePlayer_2], properties {weights [1,1,1]}}
 {id=WavePlayer_0, properties {soundWaveAssetPtr=/Game/Sounds/Fort_Monster_Storm_King/SFX/Final_SFX/StormKing_Horn_Break_v4_01}}
 {id=WavePlayer_1, properties {soundWaveAssetPtr=/Game/Sounds/Fort_Monster_Storm_King/SFX/Final_SFX/StormKing_Horn_Break_v4_02}}
 {id=WavePlayer_2, properties {soundWaveAssetPtr=/Game/Sounds/Fort_Monster_Storm_King/SFX/Final_SFX/StormKing_Horn_Break_v4_03}}
]
properties {
 bOverrideAttenuation=true
 maxDistance=2097152
 soundClassObject=&Fort_2DHitNotifications_Instigate_PSM
 volumeMultiplier=0.45
 attenuationOverrides {bAttenuate=false, falloffDistance=10000, stereoSpread=500, attenuationShapeExtents {x=5000}}
 concurrencyOverrides {maxCount=2, resolutionRule=PreventNew}
}
