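# DadBro horn impact sound cue that layers a Storm King horn hit with an enemy flesh impact.
# A mixer plays one of three random horn impacts alongside one of four random stereo player impacts, then a modulator slightly varies pitch.
# It plays unattenuated at reduced volume through a 2D hit-notification instigator sound class, with limited concurrency that stops the oldest instance.
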SoundCue /Game/Sounds/Apollo/Events/Dadbro/DadBro_Horn_Impact_Cue
root=Modulator_0
nodes [
 {id=Mixer_0, children [Random_0,Random_3], properties {inputVolume [1,1]}}
 {id=Modulator_0, children [Mixer_0], properties {pitchMax=1, pitchMin=0.98}}
 {id=Random_0, children [WavePlayer_2,WavePlayer_3,WavePlayer_4], properties {weights [1,1,1]}}
 {id=Random_3, children [WavePlayer_10, WavePlayer_11, WavePlayer_12, WavePlayer_13], properties {weights [1,1,1,1]}}
 {id=WavePlayer_10, properties {soundWaveAssetPtr=/Game/Sounds/Fort_Impact_Sounds/Flesh/player_impact_enemy_stereo_01}}
 {id=WavePlayer_11, properties {soundWaveAssetPtr=/Game/Sounds/Fort_Impact_Sounds/Flesh/player_impact_enemy_stereo_02}}
 {id=WavePlayer_12, properties {soundWaveAssetPtr=/Game/Sounds/Fort_Impact_Sounds/Flesh/player_impact_enemy_stereo_03}}
 {id=WavePlayer_13, properties {soundWaveAssetPtr=/Game/Sounds/Fort_Impact_Sounds/Flesh/player_impact_enemy_stereo_04}}
 {id=WavePlayer_2, properties {soundWaveAssetPtr=/Game/Sounds/Fort_Monster_Storm_King/SFX/Final_SFX/StormKing_Horn_Impacts_v2_01}}
 {id=WavePlayer_3, properties {soundWaveAssetPtr=/Game/Sounds/Fort_Monster_Storm_King/SFX/Final_SFX/StormKing_Horn_Impacts_v2_02}}
 {id=WavePlayer_4, properties {soundWaveAssetPtr=/Game/Sounds/Fort_Monster_Storm_King/SFX/Final_SFX/StormKing_Horn_Impacts_v2_03}}
]
properties {
 bOverrideAttenuation=true
 bOverrideConcurrency=true
 maxDistance=2097152
 soundClassObject=&Fort_2DHitNotifications_Instigate_PSM
 volumeMultiplier=0.3
 attenuationOverrides {bAttenuate=false, falloffDistance=10000, attenuationShapeExtents {x=0}}
 concurrencyOverrides {maxCount=5, resolutionRule=StopOldest}
}
