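# DadBro sound cue for the lightning-summons ground smash in a boss fight, reusing Storm King sound effects.
# A Random node picks one of three equally weighted ground smash waves and feeds it through a Modulator.
# It plays in the boss fight SFX class, with its own long-range attenuation and a concurrency override that blocks new instances at the limit.
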SoundCue /Game/Sounds/Apollo/Events/Dadbro/DadBro_LightningSummons_GroundSmash_Cue
root=Modulator_3
nodes [
 {id=Modulator_3, children [Random_0]}
 {id=Random_0, children [WavePlayer_0,WavePlayer_1,WavePlayer_2], properties {weights [1,1,1]}}
 {id=WavePlayer_0, properties {soundWaveAssetPtr=/Game/Sounds/Fort_Monster_Storm_King/SFX/Final_SFX/StormKing_LightningSummons_GroundSmash_01}}
 {id=WavePlayer_1, properties {soundWaveAssetPtr=/Game/Sounds/Fort_Monster_Storm_King/SFX/Final_SFX/StormKing_LightningSummons_GroundSmash_02}}
 {id=WavePlayer_2, properties {soundWaveAssetPtr=/Game/Sounds/Fort_Monster_Storm_King/SFX/Final_SFX/StormKing_LightningSummons_GroundSmash_03}}
]
properties {
 bOverrideAttenuation=true
 bOverrideConcurrency=true
 maxDistance=30000
 soundClassObject=&Fort_BossFight_SFX
 volumeMultiplier=1
 attenuationOverrides {falloffDistance=20000, attenuationShapeExtents {x=10000}}
 concurrencyOverrides {resolutionRule=PreventNew}
}
