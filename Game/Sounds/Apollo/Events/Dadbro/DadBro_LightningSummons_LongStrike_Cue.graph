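# Sound cue for the long lightning strike in the DadBro boss's lightning summons, reusing Storm King lightning loop waves on the boss-fight SFX class.
# A Random node picks one of three equally weighted waves, and an Enveloper holds it near full volume before fading it out by about four seconds.
# Overrides disable distance attenuation and allow only one instance at a time, preventing new plays while one is active.
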SoundCue /Game/Sounds/Apollo/Events/Dadbro/DadBro_LightningSummons_LongStrike_Cue
root=Enveloper_0
nodes [
 {
  id=Enveloper_0
  children [Random_0]
  properties {
   pitchCurve {editorCurveData {keys [{time=0}]}}
   volumeCurve {editorCurveData {keys [<at 0 1>,<at 2.8243 1.0021>,<at 3.0943 0.5893>,<at 3.8441 -0.0042>]}}
  }
 }
 {id=Random_0, children [WavePlayer_0,WavePlayer_1,WavePlayer_2], properties {weights [1,1,1]}}
 {
  id=WavePlayer_0
  properties {soundWaveAssetPtr=/Game/Sounds/Fort_Monster_Storm_King/SFX/Final_SFX/StormKing_LightningSummons_Lightning_Prct_Loop_01}
 }
 {
  id=WavePlayer_1
  properties {soundWaveAssetPtr=/Game/Sounds/Fort_Monster_Storm_King/SFX/Final_SFX/StormKing_LightningSummons_Lightning_Prct_Loop_02}
 }
 {
  id=WavePlayer_2
  properties {soundWaveAssetPtr=/Game/Sounds/Fort_Monster_Storm_King/SFX/Final_SFX/StormKing_LightningSummons_Lightning_Prct_Loop_03}
 }
]
properties {
 bOverrideAttenuation=true
 bOverrideConcurrency=true
 maxDistance=2097152
 soundClassObject=&Fort_BossFight_SFX
 volumeMultiplier=1.25
 attenuationOverrides {
  bAttenuate=false
  bEnableFocusInterpolation=true
  falloffDistance=20000
  focusAttackInterpSpeed=0.1
  focusAzimuth=15
  focusReleaseInterpSpeed=0.5
  nonFocusAzimuth=30
  nonFocusVolumeAttenuation=0
  occlusionInterpolationTime=2
  occlusionTraceChannel=ECC_GameTraceChannel11
  occlusionVolumeAttenuation=0
  spatializationAlgorithm=SPATIALIZATION_HRTF
  stereoSpread=500
  attenuationShapeExtents {x=10000}
 }
 concurrencyOverrides {maxCount=1, resolutionRule=PreventNew}
}
