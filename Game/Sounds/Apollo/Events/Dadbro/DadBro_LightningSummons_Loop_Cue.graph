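# Looping sound cue for the DadBro boss fight's lightning summons, reusing Storm King lightning loop waves.
# A fixed one-second delay leads into a modulator and an equal-weight random pick among three looping wave players.
# It plays on the boss-fight SFX class with its own long-range attenuation override.
# A concurrency override allows only one instance at a time and blocks new ones while it plays.
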SoundCue /Game/Sounds/Apollo/Events/Dadbro/DadBro_LightningSummons_Loop_Cue
root=Delay_0
nodes [
 {id=Delay_0, children [Modulator_0], properties {delayMax=1, delayMin=1}}
 {id=Modulator_0, children [Random_0]}
 {id=Random_0, children [WavePlayer_0,WavePlayer_1,WavePlayer_2], properties {weights [1,1,1]}}
 {
  id=WavePlayer_0
  properties {bLooping=true, soundWaveAssetPtr=/Game/Sounds/Fort_Monster_Storm_King/SFX/Final_SFX/StormKing_LightningSummons_Lightning_Prct_Loop_01}
 }
 {
  id=WavePlayer_1
  properties {bLooping=true, soundWaveAssetPtr=/Game/Sounds/Fort_Monster_Storm_King/SFX/Final_SFX/StormKing_LightningSummons_Lightning_Prct_Loop_02}
 }
 {
  id=WavePlayer_2
  properties {bLooping=true, soundWaveAssetPtr=/Game/Sounds/Fort_Monster_Storm_King/SFX/Final_SFX/StormKing_LightningSummons_Lightning_Prct_Loop_03}
 }
]
properties {
 bOverrideAttenuation=true
 bOverrideConcurrency=true
 maxDistance=30000
 soundClassObject=&Fort_BossFight_SFX
 attenuationOverrides {
  bEnableFocusInterpolation=true
  falloffDistance=20000
  focusAttackInterpSpeed=0.1
  focusAzimuth=15
  focusReleaseInterpSpeed=0.5
  nonFocusAzimuth=30
  nonFocusVolumeAttenuation=0
  occlusionInterpolationTime=2
  occlusionTraceChannel=ECC_GameTraceChannel11
  occlusionVolumeAttenuation=0
  stereoSpread=500
  attenuationShapeExtents {x=10000}
 }
 concurrencyOverrides {maxCount=1, resolutionRule=PreventNew}
}
