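# Sound cue for the short lightning summons effect in the DadBro boss fight event, built from Storm King lightning impact sounds.
# A mixer layers three random picks from the same three impact waves: one immediate, one after a fixed half-second delay, and one after a slightly later randomized delay.
# Both delayed layers pass through modulators at default settings.
# Overrides disable distance attenuation and set concurrency to prevent new instances at the limit.
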SoundCue /Game/Sounds/Apollo/Events/Dadbro/DadBro_LightningSummons_Short_Cue
root=Mixer_0
nodes [
 {id=Delay_0, children [Random_2], properties {delayMax=0.75, delayMin=0.65}}
 {id=Delay_1, children [Random_4], properties {delayMax=0.5, delayMin=0.5}}
 {id=Mixer_0, children [Random_1,Modulator_3,Modulator_0], properties {inputVolume [1,1,1]}}
 {id=Modulator_0, children [Delay_0]}
 {id=Modulator_3, children [Delay_1]}
 {id=Random_1, children [WavePlayer_5,WavePlayer_6,WavePlayer_7], properties {weights [1,1,1]}}
 {id=Random_2, children [WavePlayer_8, WavePlayer_9, WavePlayer_10], properties {weights [1,1,1]}}
 {id=Random_4, children [WavePlayer_4, WavePlayer_18, WavePlayer_19], properties {weights [1,1,1]}}
 {
  id=WavePlayer_10
  properties {soundWaveAssetPtr=/Game/Sounds/Fort_Monster_Storm_King/SFX/Final_SFX/StormKing_LightningSummons_Lightning_Impact_04}
 }
 {
  id=WavePlayer_18
  properties {soundWaveAssetPtr=/Game/Sounds/Fort_Monster_Storm_King/SFX/Final_SFX/StormKing_LightningSummons_Lightning_Impact_03}
 }
 {
  id=WavePlayer_19
  properties {soundWaveAssetPtr=&StormKing_LightningSummons_Lightning_Impact_04}
 }
 {
  id=WavePlayer_4
  properties {soundWaveAssetPtr=/Game/Sounds/Fort_Monster_Storm_King/SFX/Final_SFX/StormKing_LightningSummons_Lightning_Impact_02}
 }
 {
  id=WavePlayer_5
  properties {soundWaveAssetPtr=&StormKing_LightningSummons_Lightning_Impact_02}
 }
 {
  id=WavePlayer_6
  properties {soundWaveAssetPtr=&StormKing_LightningSummons_Lightning_Impact_03}
 }
 {
  id=WavePlayer_7
  properties {soundWaveAssetPtr=&StormKing_LightningSummons_Lightning_Impact_04}
 }
 {
  id=WavePlayer_8
  properties {soundWaveAssetPtr=&StormKing_LightningSummons_Lightning_Impact_02}
 }
 {
  id=WavePlayer_9
  properties {soundWaveAssetPtr=&StormKing_LightningSummons_Lightning_Impact_03}
 }
]
properties {
 bOverrideAttenuation=true
 bOverrideConcurrency=true
 maxDistance=2097152
 soundClassObject=&Fort_BossFight_SFX
 volumeMultiplier=0.5
 attenuationOverrides {bAttenuate=false, falloffDistance=20000, stereoSpread=1000, attenuationShapeExtents {x=10000}}
 concurrencyOverrides {resolutionRule=PreventNew}
}
