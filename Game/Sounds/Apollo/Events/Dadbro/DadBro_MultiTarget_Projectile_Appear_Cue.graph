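# Sound cue for the appearance of DadBro multi-target projectiles, reusing Storm King lightning projectile sounds.
# A quality-level root feeds an equal-weight random pick of three wave variations on its first two levels, leaving the other two empty.
# It uses close-projectile attenuation, the boss-fight player-critical sound class, half volume and a concurrency override that limits simultaneous voices.
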SoundCue /Game/Sounds/Apollo/Events/Dadbro/DadBro_MultiTarget_Projectile_Appear_Cue
root=QualityLevel_0
nodes [
 {id=QualityLevel_0, children [Random_0,Random_0,null,null]}
 {id=Random_0, children [WavePlayer_2,WavePlayer_3,WavePlayer_4], properties {weights [1,1,1]}}
 {id=WavePlayer_2, properties {soundWaveAssetPtr=/Game/Sounds/Fort_Monster_Storm_King/SFX/Final_SFX/StormKing_Projectile_Lightning_Appear_01}}
 {id=WavePlayer_3, properties {soundWaveAssetPtr=/Game/Sounds/Fort_Monster_Storm_King/SFX/Final_SFX/StormKing_Projectile_Lightning_Appear_02}}
 {id=WavePlayer_4, properties {soundWaveAssetPtr=/Game/Sounds/Fort_Monster_Storm_King/SFX/Final_SFX/StormKing_Projectile_Lightning_Appear_03}}
]
properties {
 attenuationSettings=/Game/Sounds/Apollo/Events/Dadbro/Attenuation/Dadbro_MultiTarget_Projectiles_Close_Att
 bOverrideConcurrency=true
 maxDistance=5750
 soundClassObject=/Game/Sounds/FortSoundClassesAndModes/Fort_BossFight_SoundClass/Fort_BossFight_PlayerCritical
 volumeMultiplier=0.5
 attenuationOverrides {falloffDistance=8000, attenuationShapeExtents {x=0}}
 concurrencyOverrides {maxCount=2, voiceStealReleaseTime=1}
}
