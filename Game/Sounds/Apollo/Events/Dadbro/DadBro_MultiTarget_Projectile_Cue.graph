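# Sound cue for DadBro boss fight multi-target projectiles, playing a mono StormKing rock incoming whistle.
# The wave passes through a Modulator, an Attenuation node and a SoundClass node overriding to boss-fight PlayerCritical, then into a Mixer with two empty inputs.
# The cue itself uses the boss-fight Projectile class, shares the node's attenuation asset, and overrides concurrency to cap simultaneous instances.
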SoundCue /Game/Sounds/Apollo/Events/Dadbro/DadBro_MultiTarget_Projectile_Cue
root=Mixer_3
nodes [
 {
  id=Attenuation_9
  children [Modulator_6]
  properties {
   attenuationSettings=/Game/Sounds/Apollo/Events/Dadbro/Attenuation/Dadbro_MultiTarget_Projectiles_Att
   attenuationOverrides {bAttenuateWithLPF=true, falloffDistance=5000, lPFRadiusMax=3000, lPFRadiusMin=1500, attenuationShapeExtents {x=2000}}
  }
 }
 {id=Mixer_3, children [SoundClass_7,null,null], properties {inputVolume [1,0.7,1]}}
 {id=Modulator_6, children [WavePlayer_1]}
 {
  id=SoundClass_7
  children [Attenuation_9]
  properties {soundClassOverride=/Game/Sounds/FortSoundClassesAndModes/Fort_BossFight_SoundClass/Fort_BossFight_PlayerCritical}
 }
 {id=WavePlayer_1, properties {soundWaveAssetPtr=&/StormKing_Rock_Incoming_Whistle_01_Mono}}
]
properties {
 attenuationSettings=&Dadbro_MultiTarget_Projectiles_Att
 bOverrideConcurrency=true
 maxDistance=4000
 soundClassObject=/Game/Sounds/FortSoundClassesAndModes/Fort_BossFight_SoundClass/Fort_BossFight_Projectile
 volumeMultiplier=0.8
 attenuationOverrides {bAttenuate=false, falloffDistance=4000, attenuationShapeExtents {x=1000}}
 concurrencyOverrides {maxCount=3}
}
