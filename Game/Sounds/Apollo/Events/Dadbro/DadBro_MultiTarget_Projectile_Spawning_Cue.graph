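# Sound cue for the spawning of DadBro boss multi-target projectiles, reusing a Storm King projectile lightning summon sound.
# A single wave player feeds a Modulator node left at its usual settings.
# It routes to the boss-fight projectile sound class and references a distant-projectile attenuation asset.
# Overridden concurrency caps simultaneous copies at a small count, with a release time when voices are stolen.
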SoundCue /Game/Sounds/Apollo/Events/Dadbro/DadBro_MultiTarget_Projectile_Spawning_Cue
root=Modulator_0
nodes [
 {id=Modulator_0, children [WavePlayer_2]}
 {id=WavePlayer_2, properties {soundWaveAssetPtr=/Game/Sounds/Fort_Monster_Storm_King/SFX/Final_SFX/StormKing_Projectile_Lightning_Summon_01}}
]
properties {
 attenuationSettings=/Game/Sounds/Apollo/Events/Dadbro/Attenuation/Dadbro_MultiTarget_Projectiles_Distant_Att
 bOverrideConcurrency=true
 maxDistance=7000
 soundClassObject=/Game/Sounds/FortSoundClassesAndModes/Fort_BossFight_SoundClass/Fort_BossFight_Projectile
 volumeMultiplier=0.8
 attenuationOverrides {falloffDistance=10000, spatializationAlgorithm=SPATIALIZATION_HRTF, attenuationShapeExtents {x=0}}
 concurrencyOverrides {maxCount=2, voiceStealReleaseTime=1}
}
