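# Sound cue for the DadBro projectile in flight, routed to the boss fight projectile sound class.
# A mixer layers a pitch-varied StormKing rock incoming whistle with one of two randomly chosen looping guided missile in-air sounds.
# The loop branch has its own attenuation node with long falloff, low-pass filtering and listener focus, separate from the cue's attenuation asset.
# A concurrency override blocks new instances once the limit is reached.
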SoundCue /Game/Sounds/Apollo/Events/Dadbro/DadBro_Projectile_Cue
root=Mixer_0
nodes [
 {
  id=Attenuation_0
  children [Modulator_1]
  properties {
   bOverrideAttenuation=true
   attenuationOverrides {bAttenuateWithLPF=true, bEnableListenerFocus=true, falloffDistance=30000, nonFocusVolumeAttenuation=0.6, attenuationShapeExtents {x=0}}
  }
 }
 {id=Mixer_0, children [Modulator_0, Attenuation_0], properties {inputVolume [1,1]}}
 {id=Modulator_0, children [WavePlayer_1], properties {pitchMax=1.15, pitchMin=1}}
 {id=Modulator_1, children [Random_1]}
 {id=Random_1, children [WavePlayer_8,WavePlayer_9], properties {weights [1,1]}}
 {id=WavePlayer_1, properties {soundWaveAssetPtr=&/StormKing_Rock_Incoming_Whistle_01_Mono}}
 {
  id=WavePlayer_8
  properties {bLooping=true, soundWaveAssetPtr=/Game/Athena/Sounds/Weapons/Guided_Missile/GuidedMissile_InAir_Banking_Loop_02}
 }
 {
  id=WavePlayer_9
  properties {bLooping=true, soundWaveAssetPtr=/Game/Athena/Sounds/Weapons/Guided_Missile/GuidedMissile_InAir_Banking_Loop_03}
 }
]
properties {
 attenuationSettings=/Game/Sounds/Apollo/Events/Dadbro/Attenuation/Dadbro_Projectile_Att
 bOverrideConcurrency=true
 maxDistance=30000
 soundClassObject=/Game/Sounds/FortSoundClassesAndModes/Fort_BossFight_SoundClass/Fort_BossFight_Projectile
 volumeMultiplier=1
 attenuationOverrides {falloffDistance=5000, attenuationShapeExtents {x=0}}
 concurrencyOverrides {resolutionRule=PreventNew}
}
