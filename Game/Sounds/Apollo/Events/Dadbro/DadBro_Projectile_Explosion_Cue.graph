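# Sound cue for DadBro projectile explosions in the Apollo event sounds, reusing Storm King projectile impact waves.
# An equally weighted random pick of three waves passes through a short random delay and a default Modulator for variation.
# It routes to the boss-fight projectile sound class with overridden attenuation using HRTF spatialization.
# Its concurrency override caps simultaneous instances and blocks new ones at the limit.
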SoundCue /Game/Sounds/Apollo/Events/Dadbro/DadBro_Projectile_Explosion_Cue
root=Modulator_1
nodes [
 {id=Delay_1, children [Random_0], properties {delayMax=0.05}}
 {id=Modulator_1, children [Delay_1]}
 {id=Random_0, children [WavePlayer_1,WavePlayer_2,WavePlayer_3], properties {bShouldExcludeFromBranchCulling=true, weights [1,1,1]}}
 {id=WavePlayer_1, properties {soundWaveAssetPtr=/Game/Sounds/Fort_Monster_Storm_King/SFX/Final_SFX/StormKing_Projectile_Impact_01}}
 {id=WavePlayer_2, properties {soundWaveAssetPtr=/Game/Sounds/Fort_Monster_Storm_King/SFX/Final_SFX/StormKing_Projectile_Impact_02}}
 {id=WavePlayer_3, properties {soundWaveAssetPtr=/Game/Sounds/Fort_Monster_Storm_King/SFX/Final_SFX/StormKing_Projectile_Impact_03}}
]
properties {
 bOverrideAttenuation=true
 bOverrideConcurrency=true
 maxDistance=4200
 soundClassObject=/Game/Sounds/FortSoundClassesAndModes/Fort_BossFight_SoundClass/Fort_BossFight_Projectile
 volumeMultiplier=1.25
 attenuationOverrides {spatializationAlgorithm=SPATIALIZATION_HRTF, stereoSpread=512, attenuationShapeExtents {x=600}}
 concurrencyOverrides {maxCount=4, resolutionRule=PreventNew}
}
