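# Looping storm ambience cue for DadBro's "StormPants", routed to the boss fight foley sound class.
# Mixes a looping Storm King tornado body with a half-volume lightning layer that endlessly picks among four spawn sounds after short random delays.
# The lightning branch overrides attenuation with a wider radius and stronger non-focus reduction than the cue's own settings.
# Concurrency allows one instance, rejecting new ones; only the first two quality-level slots are connected.
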SoundCue /Game/Sounds/Apollo/Events/Dadbro/DadBro_StormPants_Loop_Cue
root=QualityLevel_0
nodes [
 {
  id=Attenuation_2
  children [Modulator_1]
  properties {
   bOverrideAttenuation=true
   attenuationOverrides {bEnableListenerFocus=true, falloffDistance=2000, nonFocusDistanceScale=0.3, nonFocusVolumeAttenuation=0.1, attenuationShapeExtents {x=2000}}
  }
 }
 {id=Delay_0, children [Random_0], properties {delayMax=0.2}}
 {id=Looping_1, children [Delay_0]}
 {id=Mixer_2, children [Modulator_0, Attenuation_2], properties {inputVolume [1,0.5]}}
 {id=Modulator_0, children [WavePlayer_2], properties {volumeMax=1, volumeMin=1}}
 {id=Modulator_1, children [Looping_1]}
 {id=QualityLevel_0, children [Mixer_2,Mixer_2,null,null]}
 {id=Random_0, children [WavePlayer_10, WavePlayer_11, WavePlayer_12, WavePlayer_14], properties {weights [1,1,1,1]}}
 {id=WavePlayer_10, properties {soundWaveAssetPtr=/Game/Sounds/Fort_GamePlay_Sounds/Storms/lightning_spawn_06}}
 {id=WavePlayer_11, properties {soundWaveAssetPtr=/Game/Sounds/Fort_GamePlay_Sounds/Storms/lightning_spawn_08}}
 {id=WavePlayer_12, properties {soundWaveAssetPtr=/Game/Sounds/Fort_GamePlay_Sounds/Storms/lightning_spawn_09}}
 {id=WavePlayer_14, properties {soundWaveAssetPtr=/Game/Sounds/Fort_GamePlay_Sounds/Storms/lightning_spawn_03}}
 {
  id=WavePlayer_2
  properties {bLooping=true, soundWaveAssetPtr=/Game/Sounds/Fort_Monster_Storm_King/SFX/Final_SFX/StormKing_Tornado_BodyStorm_Loop_01}
 }
]
properties {
 bOverrideAttenuation=true
 bOverrideConcurrency=true
 maxDistance=4000
 soundClassObject=/Game/Sounds/FortSoundClassesAndModes/Fort_BossFight_SoundClass/Fort_BossFight_Foley
 volumeMultiplier=1
 attenuationOverrides {
  bEnableListenerFocus=true
  falloffDistance=500
  focusAttackInterpSpeed=0.1
  focusAzimuth=60
  focusReleaseInterpSpeed=0.5
  nonFocusAzimuth=90
  nonFocusDistanceScale=0.7
  nonFocusVolumeAttenuation=0.5
  occlusionInterpolationTime=2
  occlusionTraceChannel=ECC_GameTraceChannel11
  occlusionVolumeAttenuation=0
  stereoSpread=1000
  attenuationShapeExtents {x=1000}
 }
 concurrencyOverrides {maxCount=1, resolutionRule=PreventNew}
}
