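# Sound cue for the DadBro sweep beam impact, reusing a Storm King beam attack impact wave.
# A single wave player feeds a modulator that randomizes pitch, with a raised upper limit.
# The cue plays with boosted volume in the boss fight player-critical sound class.
# Its attenuation override turns off distance-based volume falloff, so loudness does not drop with range.
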SoundCue /Game/Sounds/Apollo/Events/Dadbro/DadBro_Sweep_Beam_Impact_Cue
root=Modulator_1
nodes [
 {id=Modulator_1, children [WavePlayer_5], properties {pitchMax=1.15}}
 {id=WavePlayer_5, properties {soundWaveAssetPtr=/Game/Sounds/Fort_Monster_Storm_King/SFX/StormKing_BeamAttack_Impact_02}}
]
properties {
 bOverrideAttenuation=true
 maxDistance=2097152
 soundClassObject=/Game/Sounds/FortSoundClassesAndModes/Fort_BossFight_SoundClass/Fort_BossFight_PlayerCritical
 volumeMultiplier=1.3
 attenuationOverrides {bAttenuate=false, falloffDistance=100, stereoSpread=1000, attenuationShapeExtents {x=600}}
}
