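# Sound cue for the DadBro weak-point destroyed event, reusing Storm King weak-point destruction sounds.
# Mixes two layers at equal volume, each randomly picking one of three waves: one from the v2 set, one from the original.
# Each layer passes through a default modulator and a short delay, with the original layer's delay randomized.
# Overrides attenuation to disable distance falloff and routes to a 2D hit-notification sound class.
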SoundCue /Game/Sounds/Apollo/Events/Dadbro/DadBro_WeakPoint_Destroyed_Cue
root=Mixer_0
nodes [
 {id=Delay_0, children [Modulator_0], properties {delayMax=0.25, delayMin=0.25}}
 {id=Delay_1, children [Modulator_1], properties {delayMax=0.35, delayMin=0.25}}
 {id=Mixer_0, children [Delay_0,Delay_1], properties {inputVolume [1,1]}}
 {id=Modulator_0, children [Random_1]}
 {id=Modulator_1, children [Random_2]}
 {id=Random_1, children [WavePlayer_3,WavePlayer_4,WavePlayer_5], properties {weights [1,1,1]}}
 {id=Random_2, children [WavePlayer_6,WavePlayer_7,WavePlayer_8], properties {weights [1,1,1]}}
 {id=WavePlayer_3, properties {soundWaveAssetPtr=/Game/Sounds/Fort_Monster_Storm_King/SFX/Final_SFX/StormKing_WeakPoint_Destroyed_v2_01}}
 {id=WavePlayer_4, properties {soundWaveAssetPtr=/Game/Sounds/Fort_Monster_Storm_King/SFX/Final_SFX/StormKing_WeakPoint_Destroyed_v2_02}}
 {id=WavePlayer_5, properties {soundWaveAssetPtr=/Game/Sounds/Fort_Monster_Storm_King/SFX/Final_SFX/StormKing_WeakPoint_Destroyed_v2_03}}
 {id=WavePlayer_6, properties {soundWaveAssetPtr=/Game/Sounds/Fort_Monster_Storm_King/SFX/Final_SFX/StormKing_WeakPoint_Destroyed_01}}
 {id=WavePlayer_7, properties {soundWaveAssetPtr=/Game/Sounds/Fort_Monster_Storm_King/SFX/Final_SFX/StormKing_WeakPoint_Destroyed_02}}
 {id=WavePlayer_8, properties {soundWaveAssetPtr=/Game/Sounds/Fort_Monster_Storm_King/SFX/Final_SFX/StormKing_WeakPoint_Destroyed_03}}
]
properties {
 bOverrideAttenuation=true
 maxDistance=2097152
 soundClassObject=&Fort_2DHitNotifications_Instigate_PSM
 volumeMultiplier=0.35
 attenuationOverrides {bAttenuate=false, falloffDistance=10000, stereoSpread=500, attenuationShapeExtents {x=2000}}
 concurrencyOverrides {maxCount=5, resolutionRule=PreventNew}
}
