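# Sound cue for DadBro weakpoint impacts, played as a non-attenuated hit notification limited to two instances, stopping the oldest.
# Each mix layers a briefly delayed Storm King crystal impact, a flesh impact and a pitched-down respawn point chime.
# A quality node sends its first two levels to a mix with more random variations and its last two to a leaner mix with fewer waves.
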SoundCue /Game/Sounds/Apollo/Events/Dadbro/DadBro_Weakpoint_Impact_Cue
root=QualityLevel_0
nodes [
 {id=Delay_0, children [Random_5], properties {delayMax=0.07, delayMin=0.07}}
 {id=Delay_2, children [Random_0], properties {delayMax=0.07, delayMin=0.07}}
 {id=Mixer_0, children [Delay_2,Random_3,Modulator_2], properties {inputVolume [1,0.7,1.3]}}
 {id=Mixer_1, children [Delay_0,Random_7,Modulator_3], properties {inputVolume [1,0.7,1]}}
 {id=Modulator_0, children [Mixer_1], properties {pitchMax=1, volumeMax=1, volumeMin=1}}
 {id=Modulator_1, children [Mixer_0], properties {pitchMax=1}}
 {id=Modulator_2, children [Random_2], properties {pitchMax=0.9, pitchMin=0.9}}
 {id=Modulator_3, children [WavePlayer_19], properties {pitchMax=0.9, pitchMin=0.9}}
 {id=QualityLevel_0, children [Modulator_1,Modulator_1,Modulator_0,Modulator_0]}
 {id=Random_0, children [WavePlayer_0,WavePlayer_1,WavePlayer_2,WavePlayer_3], properties {weights [1,1,1,1]}}
 {id=Random_2, children [WavePlayer_7,WavePlayer_8], properties {weights [1,1]}}
 {id=Random_3, children [WavePlayer_10, WavePlayer_11, WavePlayer_12, WavePlayer_13], properties {weights [1,1,1,1]}}
 {id=Random_5, children [WavePlayer_15, WavePlayer_16], properties {weights [1,1]}}
 {id=Random_7, children [WavePlayer_21, WavePlayer_22], properties {weights [1,1]}}
 {id=WavePlayer_0, properties {soundWaveAssetPtr=/Game/Sounds/Fort_Monster_Storm_King/SFX/Final_SFX/StormKing_CrystalWeakPoint_Impact_01}}
 {id=WavePlayer_1, properties {soundWaveAssetPtr=/Game/Sounds/Fort_Monster_Storm_King/SFX/Final_SFX/StormKing_CrystalWeakPoint_Impact_02}}
 {id=WavePlayer_10, properties {soundWaveAssetPtr=/Game/Sounds/Fort_Impact_Sounds/Flesh/player_impact_enemy_stereo_01}}
 {id=WavePlayer_11, properties {soundWaveAssetPtr=/Game/Sounds/Fort_Impact_Sounds/Flesh/player_impact_enemy_stereo_02}}
 {id=WavePlayer_12, properties {soundWaveAssetPtr=/Game/Sounds/Fort_Impact_Sounds/Flesh/player_impact_enemy_stereo_03}}
 {id=WavePlayer_13, properties {soundWaveAssetPtr=/Game/Sounds/Fort_Impact_Sounds/Flesh/player_impact_enemy_stereo_04}}
 {id=WavePlayer_15, properties {soundWaveAssetPtr=&StormKing_CrystalWeakPoint_Impact_01}}
 {id=WavePlayer_16, properties {soundWaveAssetPtr=&StormKing_CrystalWeakPoint_Impact_02}}
 {id=WavePlayer_19, properties {soundWaveAssetPtr=/Game/Athena/Sounds/LTMs/Respawn/LTM_Respawn_Point_Increase_04}}
 {id=WavePlayer_2, properties {soundWaveAssetPtr=/Game/Sounds/Fort_Monster_Storm_King/SFX/Final_SFX/StormKing_CrystalWeakPoint_Impact_03}}
 {id=WavePlayer_21, properties {soundWaveAssetPtr=&player_impact_enemy_stereo_01}}
 {id=WavePlayer_22, properties {soundWaveAssetPtr=&player_impact_enemy_stereo_02}}
 {id=WavePlayer_3, properties {soundWaveAssetPtr=/Game/Sounds/Fort_Monster_Storm_King/SFX/Final_SFX/StormKing_CrystalWeakPoint_Impact_04}}
 {id=WavePlayer_7, properties {soundWaveAssetPtr=&LTM_Respawn_Point_Increase_04}}
 {id=WavePlayer_8, properties {soundWaveAssetPtr=/Game/Athena/Sounds/LTMs/Respawn/LTM_Respawn_Point_Increase_05}}
]
properties {
 bOverrideAttenuation=true
 bOverrideConcurrency=true
 maxDistance=2097152
 soundClassObject=&Fort_2DHitNotifications_Instigate_PSM
 volumeMultiplier=0.2
 attenuationOverrides {bAttenuate=false, falloffDistance=10000, attenuationShapeExtents {x=0}}
 concurrencyOverrides {maxCount=2, resolutionRule=StopOldest}
}
