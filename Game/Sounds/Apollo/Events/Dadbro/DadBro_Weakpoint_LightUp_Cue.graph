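# Sound cue for the DadBro weakpoint light-up event, reusing a Storm King crystal weak point light-up wave.
# A single wave player feeds a modulator that varies the sound on each play, with the pitch minimum raised to 1.
# It routes to a 2D hit-notification sound class at reduced volume, with overrides that disable distance attenuation and cap concurrent plays, preventing new ones beyond the limit.
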SoundCue /Game/Sounds/Apollo/Events/Dadbro/DadBro_Weakpoint_LightUp_Cue
root=Modulator_1
nodes [
 {id=Modulator_1, children [WavePlayer_6], properties {pitchMin=1}}
 {id=WavePlayer_6, properties {soundWaveAssetPtr=&/StormKing_Crystal_WeakPoint_LightUp_v3}}
]
properties {
 bOverrideAttenuation=true
 bOverrideConcurrency=true
 maxDistance=2097152
 soundClassObject=&Fort_2DHitNotifications_Instigate_PSM
 volumeMultiplier=0.28
 attenuationOverrides {bAttenuate=false, falloffDistance=10000, attenuationShapeExtents {x=0}}
 concurrencyOverrides {maxCount=5, resolutionRule=PreventNew}
}
