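# Sound cue for the body-hit impact sound of Dadbro, classed as an instigator-side 2D hit notification.
# It randomly picks one of two equally weighted impact waves, passes it through a default Modulator for variation, then applies a short fixed delay before playback.
# It plays at reduced volume under a shared hit-notification concurrency set named for a two-voice, stop-oldest limit with volume scaling.
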SoundCue /Game/Sounds/Apollo/Events/Dadbro/Dadbro_BodyHit_Impact_Cue
root=Delay_0
nodes [
 {id=Delay_0, children [Modulator_0], properties {delayMax=0.07, delayMin=0.07}}
 {id=Modulator_0, children [Random_5]}
 {id=Random_5, children [WavePlayer_3,WavePlayer_4], properties {weights [1,1]}}
 {id=WavePlayer_3, properties {soundWaveAssetPtr=&/Dadbro_BodyHit_Impact_01}}
 {id=WavePlayer_4, properties {soundWaveAssetPtr=&/Dadbro_BodyHit_Impact_02}}
]
properties {
 maxDistance=2097152
 soundClassObject=&Fort_2DHitNotifications_Instigate_PSM
 volumeMultiplier=0.2
 concurrencyOverrides {maxCount=2, resolutionRule=StopOldest}
 concurrencySet [&HitNotifications_TwoMaximum_StopOldest_VolumeScale]
}
